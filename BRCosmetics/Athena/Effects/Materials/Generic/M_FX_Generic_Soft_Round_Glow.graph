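# Unlit translucent particle material that draws a procedural soft round glow for sprite, mesh and Niagara effects.
# A centered SphereMask with adjustable radius, hardness and power exponent forms the shape without textures.
# Particle color, optionally tinted by time-of-day colors, is multiplied by the mask and passed through an emissive adjustment function.
# Mask times particle alpha gives opacity, with depth fade applied on the default shading path but skipped on mobile.
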Material /BRCosmetics/Athena/Effects/Materials/Generic/M_FX_Generic_Soft_Round_Glow
graph {
 properties {
  bUsedWithMeshParticles=true
  bUsedWithNanite=true
  bUsedWithNiagaraMeshParticles=true
  bUsedWithNiagaraSprites=true
  bUsedWithParticleSprites=true
  blendMode=BLEND_Translucent
  pixelDepthOffsetMode=PDOM_Legacy
  shadingModel=MSM_Unlit
  usedShadingModels=Unlit
 }
 nodes [
  n1 = ParticleColor()
  n2 = Multiply(
   A=MaterialFunctionCall(
    Color_Morning=VectorParameter() {group=TOD, parameterName=Morning, defaultValue=<rgba 1 1 1 1>}.RGB
    Color_Day=VectorParameter() {group=TOD, parameterName=Day, defaultValue=<rgba 1 1 1 1>}.RGB
    Color_Evening=VectorParameter() {group=TOD, parameterName=Evening, defaultValue=<rgba 1 1 1 1>}.RGB
    Color_Night=VectorParameter() {group=TOD, parameterName=Night, defaultValue=<rgba 1 1 1 1>}.RGB
   ) {materialFunction=&Color_Time_of_Day}.Result
   B=@n1.RGB
  )
  n3 = ScalarParameter() {group=TOD, parameterName=UseTimeOfDay?}
  n4 = Power(
   Base=SphereMask(
    A=TextureCoordinate()
    B=Constant2Vector() {g=0.5, r=0.5}
    Radius=ScalarParameter() {defaultValue=0.5, parameterName=Radius}
    Hardness=ScalarParameter() {parameterName=Hardness}
   )
   Exp=ScalarParameter() {defaultValue=2, parameterName=Sphere_Power_Exp}
  )
  MaterialFunctionCall(Emissive=Multiply(A=LinearInterpolate(A=@n1.RGB, B=@n2, Alpha=@n3), B=@n4)) {materialFunction=&MF_FX_EmissiveAdjustments}
  n5 = Multiply(A=@n4, B=@n1.A)
  n6 = ScalarParameter() {defaultValue=50, parameterName=FadeDistance}
  ShadingPathSwitch(Default=Multiply(A=@n5, B=DepthFade(FadeDistance=@n6)), Mobile=@n5) {desc=Migrated from FeatureLevelSwitch}
 ]
}
